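# Masked Cloth-model toon master material under the Lettuce glider chute assets, named as an erosion test.
# Lighting is built into Emissive from sky-atmosphere light direction: banded shadows, stepped metallic reflections, plus switchable rim, GGX highlight, ink-line and time-of-day stages.
# An optional painterly effect varies band edges using screen-derivative masks.
# Opacity mask erodes through a tiled texture and vertex color, smoothstepped against a Dynamic Parameter threshold.
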Material /Game/Gadgets/Assets/VinderTech_GliderChute/Glider_Lettuce/Materials/M_ToonShader_TestErosion
graph {
 properties {
  bUseMaterialAttributes=true
  bUsedWithMeshDeformer=true
  bUsedWithNiagaraMeshParticles=true
  bUsedWithSkeletalMesh=true
  blendMode=BLEND_Masked
  pixelDepthOffsetMode=PDOM_Legacy
  shadingModel=MSM_Cloth
  usedShadingModels=Cloth
 }
 nodes [
  n1 = Constant()
  n2 = TextureSampleParameter2D() {
   group="##01.Textures"
   parameterName=Texture_SSC
   samplerSource=SSM_Wrap_WorldGroupSettings
   samplerType=SAMPLERTYPE_Masks
   sortPriority=1
   texture=&T_BlackMask
  }
  n3 = ParticleColor()
  n4 = Constant() {r=1}
  n5 = If(A=@n3.A, "A > B"=@n3.A, "A == B"=@n4, "A < B"=@n4)
  n6 = VectorParameter() {group="########BasicSetting", parameterName=BaseColor, defaultValue=<rgba 1 1 1 1>}
  n7 = TextureSampleParameter2D() {
   group="##01.Textures"
   parameterName=Texture_BaseColor
   samplerType=SAMPLERTYPE_Masks
   texture=/Game/Global/Textures/Default/Blanks/T_MidgreyMask
  }
  n8 = Multiply(A=@n7.RGB, B=NamedRerouteUsage()) {constA=1}
  n9 = Multiply(
   A=NamedRerouteUsage()
   B=Power(Base=NamedRerouteUsage(), Exp=ScalarParameter() {defaultValue=0.3, group=02.Metallic_Advance, parameterName=Metallic_LightingPower}) {constExponent=0.5}
  )
  n10 = NamedRerouteUsage()
  n11 = Multiply(
   A=Multiply(A=Multiply(A=@n5, B=@n6.RGB), B=LinearInterpolate(A=@n8, B=@n9, Alpha=@n10))
   B=StaticSwitchParameter(True=NamedRerouteUsage(), False=Constant() {r=1}) {parameterName=Enable_RimRouge}
  )
  n12 = Saturate(None=@n11)
  n13 = StaticSwitchParameter(True=LinearInterpolate(A=NamedRerouteUsage(), B=@n12, Alpha=NamedRerouteUsage()), False=@n12) {parameterName=Enable_InkLine}
  n14 = NamedRerouteUsage()
  n15 = Add(A=@n13, B=@n14)
  n16 = ScalarParameter() {group="##31.Highlight", parameterName=HighLight_BlendMethod}
  n17 = StaticSwitchParameter(True=LinearInterpolate(A=@n13, B=LinearInterpolate(A=@n15, B=@n14, Alpha=@n16), Alpha=NamedRerouteUsage()), False=@n13) {parameterName=Enable_Highlight}
  n18 = NamedRerouteUsage()
  n19 = NamedRerouteUsage()
  n20 = NamedRerouteUsage()
  n21 = NamedRerouteUsage()
  n22 = Add(
   A=StaticSwitchParameter(True=LinearInterpolate(A=@n17, B=@n18, Alpha=@n19), False=@n17) {parameterName=Enable_Reflection}
   B=StaticSwitchParameter(True=Multiply(A=@n20, B=@n21), False=Constant()) {parameterName=Enable_Rimlight}
  )
  n23 = StaticSwitchParameter(
   True=LinearInterpolate(
    A=@n22
    B=MaterialFunctionCall!(Color=@n22) {materialFunction=/Game/Packages/Fortress_SharedMaterials/Anime_Master_Material/MF_TODColorCurve}.Result
    Alpha=ScalarParameter() {defaultValue=1, group=TOD_Adjustments, parameterName=TODStr}
   )
   False=@n22
  ) {defaultValue=true, parameterName=Enable_TOD}
  n24 = ScalarParameter() {defaultValue=1, group="########BasicSetting", parameterName=EmissiveScale}
  n25 = ScalarParameter() {defaultValue=1, group=Erotion, parameterName=Erosion Scale, sliderMax=10}
  n26 = TextureSampleParameter2D(UVs=Multiply(A=TextureCoordinate(), B=Divide(B=@n25))) {
   group=Erotion
   parameterName=Erosion Texture
   samplerType=SAMPLERTYPE_LinearGrayscale
   texture=/Game/Packages/Fortress_Sky/SkyDome/Master/T_SkyDome_Cloud_Voroni
  }
  n27 = VertexColor()
  n28 = LinearInterpolate(A=Multiply(A=@n26.RGBA, B=@n27.R), B=@n27.R, Alpha=ScalarParameter() {group=Erotion, parameterName=Use Vertex Color For Erosion})
  n29 = ChannelMaskParameter(None=@n28) {group=Erotion, maskChannel=Alpha, parameterName=Base Mask Channel, sortPriority=4, defaultValue=<rgba 0 0 0 1>}
  n30 = DynamicParameter() {defaultValue=<rgba 0 1 1 1>, paramNames [Erosion,Param2,Param3,Param4]}
  n31 = ScalarParameter() {defaultValue=1, parameterName=Erosion Contrast, sliderMax=10}
  n32 = ChannelMaskParameter(None=@n28) {group=Erotion, parameterName=Erosion Channel, sortPriority=-2}
  n33 = OneMinus(None=@n32.R)
  n34 = MaterialFunctionCall(
   Alpha=StaticSwitchParameter(True=Multiply(A=@n33, B=@n32.R), False=@n32.R) {group=Erotion, parameterName=Invert Erosion Values}
   Min=@n30.Erosion
   Max=Add(A=@n30.Erosion, B=OneMinus(None=@n31))
  ) {materialFunction=&SmoothStep}
  n35 = LinearInterpolate(
   A=StaticSwitchParameter(True=@n29.R, False=Constant() {r=1}) {group=Erotion, parameterName=Use Base Mask, sortPriority=3}
   B=StaticSwitchParameter(True=Multiply(A=@n29.R, B=@n34.Result), False=@n34.Result) {group=Erotion, parameterName=Mutiply by Base Alpha, sortPriority=3}
   Alpha=ScalarParameter() {defaultValue=1, group=Erotion, parameterName=Erosion Influence, sortPriority=-3}
  )
  n36 = TextureSampleParameter2D() {
   group="##01.Textures"
   parameterName=Texture_Normal
   samplerType=SAMPLERTYPE_Normal
   texture=&FlatNormalBlue
  }
  n37 = ScalarParameter() {defaultValue=0.003, group=DontModify, parameterName=Subsurface Mask Value, sliderMax=1}
  MakeMaterialAttributes!(
   BaseColor=@n1
   Metallic=StaticSwitchParameter(True=@n2.R, False=@n1) {parameterName=Enable_AdvanceTextureSetting}
   Specular=@n1
   Roughness=StaticSwitchParameter(True=@n2.G, False=@n1) {parameterName=Enable_AdvanceTextureSetting}
   Anisotropy
   EmissiveColor=MaterialFunctionCall(Emissive=Multiply(A=@n23, B=@n24)) {materialFunction=&MF_FX_EmissiveAdjustments}.Result
   Opacity=Constant() {r=1}
   OpacityMask=StaticSwitchParameter(True=Saturate(None=@n35), False=Constant() {r=1}) {defaultValue=true, group=Erotion, parameterName=Erosion, sortPriority=-10}
   Normal=MaterialFunctionCall(Normal=@n36.RGB, Flatness=ScalarParameter() {group="########BasicSetting", parameterName=NormalFlatness}) {materialFunction=&FlattenNormal}.Result
   Tangent
   WorldPositionOffset
   SubsurfaceColor=MaterialFunctionCall(Y=@n37, Z=@n37) {materialFunction=&MakeFloat3}.Result
   ClearCoat
   ClearCoatRoughness
   AmbientOcclusion=StaticSwitchParameter(True=NamedRerouteUsage(), False=Constant()) {parameterName=Enable_InkLine}
   Refraction
   CustomizedUVs_0
   CustomizedUVs_1
   CustomizedUVs_2
   CustomizedUVs_3
   CustomizedUVs_4
   CustomizedUVs_5
   CustomizedUVs_6
   CustomizedUVs_7
   PixelDepthOffset
   ShadingModel
   Displacement
  )
  Multiply(A=@n11, B=StaticSwitchParameter(True=NamedRerouteUsage(), False=Constant() {r=1}) {parameterName=Enable_SphereLighting})
  NamedRerouteDeclaration(None=SkyAtmosphereLightDirection()) {name=SceneData_LightDir, nodeColor=<rgba 0.4297 0 1 1>}
  NamedRerouteDeclaration(None=CameraVectorWS()) {name=SceneData_CameraVector, nodeColor=<rgba 0.9219 1 0 1>}
  n38 = SkyLightEnvMapSample(Direction=ReflectionVectorWS() {bNormalizeCustomWorldNormal=true}, Roughness=NamedRerouteUsage())
  n39 = Desaturation(None=@n38, Fraction=Constant() {r=1})
  n40 = Normalize(VectorInput=VectorParameter() {group=02.Metallic_Advance, parameterName=Metallic_Reflection_Color, defaultValue=<rgba 1 1 1 1>}.RGB)
  n41 = LinearInterpolate(A=@n39, B=Multiply(A=@n40, B=@n39), Alpha=Saturate(None=DotProduct(A=@n38, B=@n40)))
  n42 = ScalarParameter() {defaultValue=10, group=02.Metallic_Advance, parameterName=Metallic_Step}
  n43 = Multiply(A=MaterialFunctionCall(Vector 3=@n41) {materialFunction=/Engine/Functions/Engine_MaterialFunctions02/Utility/VectorLength}.V3 Length, B=@n42) {constB=5}
  n44 = Multiply(A=Normalize(VectorInput=@n41), B=Divide(A=Ceil(None=@n43), B=@n42) {constB=5})
  n45 = Power(
   Base=NamedRerouteUsage()
   Exp=ScalarParameter() {defaultValue=30, group=02.Metallic_Advance, parameterName=Metallic_Painting_Str, sortPriority=64}
  ) {constExponent=50}
  NamedRerouteDeclaration(None=StaticSwitchParameter(True=LinearInterpolate(A=@n41, B=@n44, Alpha=@n45), False=@n44) {parameterName=Enable_PaintingEffect}) {name=Metallic_ReflectionColor, nodeColor=<rgba 0 1 0.3203 1>}
  n46 = WorldPosition()
  n47 = LinearInterpolate(
   A=ObjectPositionWS()
   B=ParticlePositionWS()
   Alpha=ScalarParameter() {defaultValue=1, group="########BasicSetting", parameterName=UseParticlePosition}
  )
  NamedRerouteDeclaration(None=Normalize(VectorInput=Subtract(A=@n46.XYZ, B=@n47))) {name=SceneData_SphereNormal, nodeColor=<rgba 0 0.7656 1 1>}
  n48 = Add(A=DotProduct(A=NamedRerouteUsage(), B=NamedRerouteUsage()))
  n49 = Saturate(None=OneMinus(None=Multiply(A=@n48) {constB=0.5}))
  n50 = ScalarParameter() {defaultValue=1, group="##11.SphereLighting", parameterName=SphereLighting_Power}
  NamedRerouteDeclaration(None=LinearInterpolate(
    B=VectorParameter() {group="##11.SphereLighting", parameterName=SphereLighting_Color, defaultValue=<rgba 2 2 2 1>}.RGB
    Alpha=Power(Base=@n49, Exp=@n50)
   ) {constA=1}) {name=SphereLighting_Color, nodeColor=<rgba 0 0.0156 1 1>}
  n51 = Add(A=NamedRerouteUsage(), B=NamedRerouteUsage())
  n52 = Transform(None=Constant3Vector() {constant=<rgba 1 0 0 0>}) {transformSourceType=TRANSFORMSOURCE_Camera}
  n53 = DotProduct(A=NamedRerouteUsage(), B=Normalize(VectorInput=Add(A=@n51, B=@n52)))
  n54 = Constant() {r=0.55}
  n55 = OneMinus(None=Saturate(None=Add(A=@n53, B=@n54)))
  n56 = NamedRerouteUsage()
  n57 = Fresnel(ExponentIn=ScalarParameter() {defaultValue=1.35, group="##21.RimRouge", parameterName=RimRouge_Fresnel}, Normal=@n56) {baseReflectFraction=0}
  NamedRerouteDeclaration(None=LinearInterpolate(
    B=VectorParameter() {group="##21.RimRouge", parameterName=RimRouge_Color, defaultValue=<rgba 2 1.55 1.4 1>}.RGB
    Alpha=Multiply(A=@n55, B=@n57)
   ) {constA=1}) {name=RimRouge_Color, nodeColor=<rgba 0.0547 1 0 1>}
  NamedRerouteDeclaration(None=VectorParameter() {group=10.Lighting_Advance, parameterName=Lighting_Color_2_LitRouge, defaultValue=<rgba 1 0.85 0.55 1>}.RGB) {name=Lighting_LitRougeColor, nodeColor=<rgba 0 1 0.1797 1>}
  NamedRerouteDeclaration(None=VectorParameter() {group=10.Lighting_Advance, parameterName=Lighting_Color_1_Lit, defaultValue=<rgba 1 1 1 1>}.RGB) {name=Lighting_LitColor, nodeColor=<rgba 0 0.0156 1 1>}
  n58 = VectorParameter() {group=10.Lighting_Advance, parameterName=Lighting_Color_3_ShadowRouge, defaultValue=<rgba 0.55 0.3644 0.3025 1>}
  NamedRerouteDeclaration(None=@n58.RGB) {name=Lighting_ShadowRougeColor, nodeColor=<rgba 0.7812 1 0 1>}
  NamedRerouteDeclaration(None=VectorParameter() {group=10.Lighting_Advance, parameterName=Lighting_Color_4_Shadow, defaultValue=<rgba 0.4 0.34 0.34 1>}.RGB) {name=Lighting_ShadowColor, nodeColor=<rgba 1 0 0.4453 1>}
  NamedRerouteDeclaration(None=VectorParameter() {group=10.Lighting_Advance, parameterName=Lighting_Color_5_FakeGI, defaultValue=<rgba 0.5 0.45 0.45 1>}.RGB) {name=Lighting_FakeGIColor, nodeColor=<rgba 0.0781 1 0 1>}
  n59 = Add(
   A=DotProduct(A=NamedRerouteUsage(), B=NamedRerouteUsage())
   B=LinearInterpolate(A=ScalarParameter() {defaultValue=-0.4, group="##10.Lighting", parameterName=Lighting_Offset}, Alpha=NamedRerouteUsage()) {constB=0}
  )
  n60 = OneMinus(None=Constant() {r=0.6})
  n61 = Constant() {r=0.01}
  n62 = PixelNormalWS()
  n63 = Saturate(None=SmoothStep(Min=@n60, Max=Add(A=@n60, B=@n61), Value=Fresnel(Normal=@n62) {baseReflectFraction=0, exponent=1}))
  n64 = Add(A=DotProduct(
    A=MaterialFunctionCall() {materialFunction=&CameraDirectionVector}.Result
    B=NamedRerouteUsage()
   )) {constB=0.25}
  n65 = StaticSwitchParameter(True=LinearInterpolate(A=@n59, B=Add(A=@n63, B=@n59), Alpha=Clamp(None=@n64) {maxDefault=0.75}), False=@n59) {parameterName=Enable_BackfaceLit}
  n66 = Saturate(None=@n65)
  n67 = Saturate(None=Multiply(A=@n65) {constB=-1})
  n68 = ScalarParameter() {defaultValue=0.1, group=10.Lighting_Advance, parameterName=Lighting_BlendWidth_RougeWithRouge}
  n69 = NamedRerouteUsage()
  n70 = StaticSwitchParameter(True=NamedRerouteUsage(), False=@n69) {parameterName=Enable_PaintingEffect}
  n71 = Multiply(A=@n68, B=@n70)
  n72 = Multiply(A=@n67, B=Divide(B=@n71))
  n73 = Saturate(None=Subtract(A=@n67, B=@n71))
  n74 = ScalarParameter() {defaultValue=0.2, group=10.Lighting_Advance, parameterName=Lighting_BlendWidth_ShadowRouge}
  n75 = Multiply(A=@n73, B=Divide(B=Multiply(A=@n74, B=@n70)))
  n76 = LinearInterpolate(
   A=LinearInterpolate(A=NamedRerouteUsage(), B=NamedRerouteUsage(), Alpha=Saturate(None=@n72))
   B=LinearInterpolate(A=NamedRerouteUsage(), B=NamedRerouteUsage(), Alpha=Saturate(None=@n75))
   Alpha=Ceil(None=@n73)
  )
  n77 = NamedRerouteUsage()
  n78 = Subtract(A=@n67, B=Multiply(A=@n68, B=@n69))
  n79 = Multiply(A=@n74, B=NamedRerouteUsage())
  n80 = Power(
   Base=Saturate(None=Subtract(A=@n78, B=@n79))
   Exp=ScalarParameter() {defaultValue=1.35, group=10.Lighting_Advance, parameterName=Lighting_FakeGIAreaPower}
  )
  n81 = ScalarParameter() {defaultValue=3, group=10.Lighting_Advance, parameterName=Lighting_GIStep}
  n82 = Divide(A=Floor(None=Multiply(A=@n80, B=@n81) {constB=10}), B=@n81) {constB=10}
  n83 = MaterialFunctionCall!(
   HSV=VectorParameter() {group="##10.Lighting", parameterName=Lighting_FinalColorHSV, defaultValue=<rgba 0 1 1 1>}.RGB
   Color=LinearInterpolate(A=LinearInterpolate(A=@n76, B=@n77, Alpha=@n82), B=NamedRerouteUsage(), Alpha=Ceil(None=@n66))
  ) {materialFunction=/Game/Effects/Niagara/MasterMaterials/Functions/Common/FXCommon_HSV}
  n84 = NamedRerouteUsage()
  n85 = Multiply(
   A=ScalarParameter() {defaultValue=0.1, group=10.Lighting_Advance, parameterName=Lighting_BlendWidth_LitRouge}
   B=StaticSwitchParameter(True=NamedRerouteUsage(), False=NamedRerouteUsage()) {parameterName=Enable_PaintingEffect}
  )
  n86 = Saturate(None=Multiply(A=@n66, B=Divide(B=@n85)))
  n87 = ScalarParameter() {defaultValue=1, group="##10.Lighting", parameterName=Lighting_Simple_LitStr}
  n88 = ScalarParameter() {defaultValue=0.5, group="##10.Lighting", parameterName=Lighting_Simple_ShadowStr}
  n89 = ScalarParameter() {defaultValue=4, group="##10.Lighting", parameterName=Lighting_Hardness}
  n90 = SmoothStep(Max=Divide(B=@n89) {constA=0.08}, Value=@n67) {constMax=0.02}
  NamedRerouteDeclaration(None=StaticSwitchParameter(True=LinearInterpolate(A=@n83.Result, B=@n84, Alpha=@n86), False=LinearInterpolate(A=@n87, B=@n88, Alpha=@n90) {constA=1, constB=0.5}) {parameterName=Enable_Lighting_ComplexBlending}) {name=Lighting_Color, nodeColor=<rgba 0.4297 1 0 1>}
  n91 = ScalarParameter() {defaultValue=0.2, group=10.Lighting_Advance, parameterName=Lighting_Painting_Hardness, sortPriority=64}
  n92 = NamedRerouteUsage()
  n93 = Constant() {r=0.2}
  n94 = Multiply(A=@n89, B=Add(A=Multiply(A=@n92, B=@n93)))
  n95 = Constant() {r=0.1}
  n96 = Power(
   Base=NamedRerouteUsage()
   Exp=LinearInterpolate(
    A=ScalarParameter() {defaultValue=50, group=10.Lighting_Advance, parameterName=Lighting_Painting_ShadowFaceStrength, sortPriority=64}
    B=ScalarParameter() {group=10.Lighting_Advance, parameterName=Lighting_Painting_LitFaceStrength, sortPriority=64}
    Alpha=MaterialFunctionCall!(In=@n65, Ramp Width=@n95) {materialFunction=/Engine/Functions/Engine_MaterialFunctions01/Math/SmoothCeil}.Result
   )
  )
  NamedRerouteDeclaration(None=Divide(B=LinearInterpolate(A=@n91, B=@n94, Alpha=@n96))) {name=Lighting_BlendWidthScaleWithPainting, nodeColor=<rgba 1 0.2812 0 1>}
  NamedRerouteDeclaration(None=Divide(B=@n94)) {name=Lighting_BlendWidthScale, nodeColor=<rgba 0.8984 1 0 1>}
  n97 = TextureSampleParameter2D() {
   group="##40.InkLine"
   parameterName=Texture_InkLines
   samplerSource=SSM_Wrap_WorldGroupSettings
   samplerType=SAMPLERTYPE_Masks
   sortPriority=1
   texture=&T_BlackMask
  }
  n98 = Max(B=@n97.R) {desc=Inject Inklines here}
  n99 = ScalarParameter() {defaultValue=15, group=40.InkLine_Advance, parameterName=Inkline_BlurPow, sliderMax=16}
  n100 = OneMinus(None=Multiply(
    A=Power(Base=@n98, Exp=@n99) {constExponent=3}
    B=ScalarParameter() {defaultValue=0.5, group=40.InkLine_Advance, parameterName=Inkline_BlurMult, sliderMax=2}
   ) {constB=0.5})
  n101 = Constant() {r=1}
  n102 = ScalarParameter() {defaultValue=100, group=40.InkLine_Advance, parameterName=Inkline_FarDistance, sliderMax=1000}
  n103 = Subtract(A=Distance(A=WorldPosition().XYZ, B=CameraPositionWS()), B=@n102)
  n104 = Subtract(A=ScalarParameter() {defaultValue=500, group=40.InkLine_Advance, parameterName=Inkline_NearDistance, sliderMax=1000}, B=@n102)
  n105 = LinearInterpolate(
   A=ScalarParameter() {defaultValue=0.02, group=40.InkLine_Advance, parameterName=Inkline_LineThickness_Near, sliderMax=1}
   B=ScalarParameter() {defaultValue=0.1, group=40.InkLine_Advance, parameterName=Inkline_LineThickness_Far, sliderMax=1}
   Alpha=Saturate(None=Divide(A=@n103, B=@n104))
  )
  n106 = ScalarParameter() {defaultValue=1, group=40.InkLine_Advance, parameterName=Inkline_LineHardness, sliderMax=1}
  n107 = OneMinus(None=Saturate(None=SphereMask(A=@n98, B=@n101, Radius=@n105, Hardness=@n106) {attenuationRadius=0.1}))
  NamedRerouteDeclaration(None=Saturate(None=Multiply(A=@n100, B=@n107))) {name=InkLine Mask, nodeColor=<rgba 0.1615 0.1615 0.1615 1>}
  NamedRerouteDeclaration(None=VectorParameter() {group="##40.InkLine", parameterName=InkLineColor, defaultValue=<rgba 0 0 0 1>}.RGB) {name=InkLine Color, nodeColor=<rgba 0.1615 0.1615 0.1615 1>}
  NamedRerouteDeclaration(None=ScalarParameter() {defaultValue=0.2, group="########BasicSetting", parameterName=Roughness}) {name=BaseParam_Roughness, nodeColor=<rgba 1 0 0.5625 1>}
  n108 = TextureSampleParameter2D() {group="##01.Textures", parameterName=Texture_MHS, samplerType=SAMPLERTYPE_Masks, texture=&T_BlackMask}
  n109 = VectorParameter() {group="########BasicSetting", parameterName=MHSTextureScale, sortPriority=64, defaultValue=<rgba 1 1 1 1>}
  n110 = MaterialFunctionCall(Float3=StaticSwitchParameter(True=Multiply(A=@n108.RGB, B=@n109.RGB), False=Constant()) {parameterName=Enable_AdvanceTextureSetting}) {materialFunction=&BreakOutFloat3Components}
  n111 = ScalarParameter() {group="########BasicSetting", parameterName=Metallic}
  NamedRerouteDeclaration(None=Saturate(None=Add(A=@n110.R, B=@n111))) {name=BaseParam_Metallic, nodeColor=<rgba 1 0 0.5625 1>}
  n112 = ScalarParameter() {defaultValue=1, group="########BasicSetting", parameterName=Specular}
  NamedRerouteDeclaration(None=Saturate(None=Add(A=@n110.B, B=@n112))) {name=BaseParam_Specular, nodeColor=<rgba 1 0 0.5625 1>}
  n113 = ScalarParameter() {defaultValue=0.3, group="########BasicSetting", parameterName=HighLight}
  NamedRerouteDeclaration(None=Saturate(None=Add(A=@n110.G, B=@n113))) {name=BaseParam_HighLight, nodeColor=<rgba 1 0 0.5625 1>}
  n114 = NamedRerouteUsage()
  n115 = DDX(Value=@n114)
  n116 = DDY(Value=@n114)
  n117 = ScalarParameter() {defaultValue=1500, group="##41.PaintingEffect", parameterName=PaintingEffect_AreaExpand}
  n118 = Distance(A=CameraPositionWS(), B=WorldPosition().XYZ)
  n119 = Divide(A=Multiply(A=@n117, B=@n117), B=Multiply(A=@n118, B=@n118))
  n120 = VectorParameter() {
   group="##41.PaintingEffect"
   parameterName=PaintingEffect_DistanceFade
   channelNames {a="", b="", g=Far, r=Near}
   defaultValue=<rgba 30 100 0 1>
  }
  n121 = Saturate(None=Divide(A=Subtract(A=@n118, B=@n120.Near) {constB=30}, B=@n120.Far) {constB=100})
  n122 = MaterialFunctionCall(Vector 3=Multiply(A=CrossProduct(A=@n115, B=@n116), B=LinearInterpolate(B=@n119, Alpha=@n121))) {materialFunction=&VectorLength}
  NamedRerouteDeclaration(None=OneMinus(None=Saturate(None=@n122.V3 Length))) {name=Painting_OriginalOutput, nodeColor=<rgba 0.7109 0 1 1>}
  NamedRerouteDeclaration(None=PixelNormalWS()) {name=SceneData_PixelNormal, nodeColor=<rgba 0 0.625 1 1>}
  n123 = ScalarParameter() {defaultValue=0.005, group="##31.Highlight", parameterName=HighLight_Softness}
  n124 = MaterialFunctionCall!(LightDirctionIn=NamedRerouteUsage(), CameraVectorIn=NamedRerouteUsage(), Roughness=NamedRerouteUsage(), NormalIn=NamedRerouteUsage()) {materialFunction=/Game/Effects/Niagara/MasterMaterials/Functions/Common/FXCommon_GGX}
  n125 = Saturate(None=SmoothStep(Max=Add(A=@n123) {constB=0.05}, Value=Saturate(None=@n124.Result)) {constMin=0.05})
  NamedRerouteDeclaration(None=@n125) {name=GGX_Mask, nodeColor=<rgba 0.1484 0 1 1>}
  NamedRerouteDeclaration(None=LinearInterpolate(
    A=VectorParameter() {group="##31.Highlight", parameterName=HighLight_Rouge, defaultValue=<rgba 1 1 1 1>}.RGB
    B=VectorParameter() {group="##31.Highlight", parameterName=HighLight_Color, defaultValue=<rgba 1 1 1 1>}.RGB
    Alpha=@n125
   )) {name=GGX_Color, nodeColor=<rgba 0.1719 0 1 1>}
  n126 = NamedRerouteUsage()
  n127 = ScalarParameter() {defaultValue=5, group="##30.Reflection", parameterName=Reflection_Step}
  n128 = Multiply(
   A=Fresnel(
    ExponentIn=ScalarParameter() {defaultValue=0.9, group="##30.Reflection", parameterName=Reflection_Painting_Fresnel, sortPriority=64}
    Normal=@n126
   ) {baseReflectFraction=0}
   B=@n127
  ) {constB=5}
  n129 = Fresnel(ExponentIn=ScalarParameter() {defaultValue=1.35, group="##30.Reflection", parameterName=Reflection_Fresnel}, Normal=@n126) {baseReflectFraction=0}
  n130 = Divide(A=Floor(None=Multiply(A=@n129, B=@n127) {constB=5}), B=@n127) {constB=5}
  n131 = LinearInterpolate(
   A=Divide(A=Floor(None=@n128), B=@n127) {constB=5}
   B=@n130
   Alpha=Power(
    Base=NamedRerouteUsage()
    Exp=ScalarParameter() {defaultValue=50, group="##30.Reflection", parameterName=Reflection_Painting_Str, sortPriority=64}
   ) {constExponent=50}
  )
  NamedRerouteDeclaration(None=Multiply(
    A=StaticSwitchParameter(True=@n131, False=@n130) {parameterName=Enable_PaintingEffect}
    B=ScalarParameter() {defaultValue=0.5, group="##30.Reflection", parameterName=Reflection_Strength}
   )) {name=Reflection_Mask, nodeColor=<rgba 1 0 0.5156 1>}
  n132 = Normalize(VectorInput=@n38)
  n133 = MaterialFunctionCall(Vector 3=@n38) {materialFunction=&VectorLength}
  n134 = Divide(A=Ceil(None=Multiply(A=@n133.V3 Length, B=@n127) {constB=5}), B=@n127) {constB=5}
  NamedRerouteDeclaration(None=Desaturation(None=Multiply(A=@n132, B=@n134), Fraction=ScalarParameter() {defaultValue=-1, group="##30.Reflection", parameterName=Reflection_Desaturation})) {name=Reflection_Color, nodeColor=<rgba 1 0 0.9375 1>}
  n135 = NamedRerouteUsage()
  n136 = NamedRerouteUsage()
  n137 = Power(Base=Saturate(None=DotProduct(A=@n135, B=@n136)))
  n138 = OneMinus(None=ScalarParameter() {defaultValue=0.3, group="##20.RimLight", parameterName=RimLight_Width})
  n139 = Add(A=ScalarParameter() {defaultValue=0.02, group="##20.RimLight", parameterName=RimLight_Softness}, B=@n138) {constB=0.5}
  n140 = DotProduct(A=PixelNormalWS(), B=CameraVectorWS())
  n141 = Saturate(None=SmoothStep(Min=@n138, Max=Saturate(None=@n139), Value=OneMinus(None=@n140)) {constMax=0.51, constMin=0.5})
  NamedRerouteDeclaration(None=Saturate(None=Multiply(A=@n137, B=@n141))) {name=RimLight_Mask, nodeColor=<rgba 1 0 0.2109 1>}
  NamedRerouteDeclaration(None=VectorParameter() {group="##20.RimLight", parameterName=RimLight_Color, defaultValue=<rgba 2 2 2 1>}.RGB) {name=RimLight_Color, nodeColor=<rgba 1 0 0.2109 1>}
  NamedRerouteUsage()
  n142 = NamedRerouteUsage()
  Normalize(VectorInput=LinearInterpolate(A=@n56, B=Add(A=@n142, B=@n56), Alpha=Constant()))
 ]
}
